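# DuckCoast animation Blueprint on the Fortnite average male skeleton, adding awareness-driven body and face animation through the Item Override layers; only ItemFullBodyOverride has logic.
# A state machine moves between cleaning idle, locomotion, alert intro/loop/outro and combat using movement, bAlerted and bInCombat.
# Locomotion and combat pass through the cached input pose; alert states layer animations over it from the spine when moving.
# AngryFaceAlpha blends an angry face through a facial-curve mask.
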AnimBlueprint /WildEstate_DuckCoast/Assets/Animation/WildEstate_DuckCoast_AnimBP
target_skeleton=&Fortnite_M_Avg_Player_Skeleton
graphs [
 AnimGraph {nodes [Output Pose()]}
 EventGraph {links [], nodes []}
 ItemAimOffset {nodes [
   Input Pose!() -> (Pose, AimOffsetAlpha:double, Yaw:double, Pitch:double) {name=InPose_AimOffset}
   Input Pose!() -> (Pose) {name=InPose_PreAimOffsets}
   Output Pose() {layerGroup=Item Override}
  ]}
 ItemFallAdditive {nodes [
   Input Pose() {name=InPoseItemFallAdditive}
   Output Pose() {layerGroup=Item Override}
  ]}
 ItemFallLandAdditive {nodes [
   Input Pose() {name=InPoseItemFallLandAdditive}
   Output Pose() {layerGroup=Item Override}
  ]}
 ItemFallLevel2Override {nodes [
   Input Pose() {name=InPose_FallLevel2}
   Output Pose() {layerGroup=Item Override}
  ]}
 ItemFinalPoseOverride {nodes [
   Input Pose() {name=InFinalPose}
   Output Pose() {layerGroup=Item Override}
  ]}
 ItemFullBodyAimOffsets {nodes [
   Input Pose!() -> (Pose, FullBody_AimOffsetAlpha:double, Aim_Yaw:double, Aim_Pitch:double) {name=InPoseFullBodyAimOffsets}
   Input Pose!() -> (Pose) {name=InPosePreFullBodyAimOffsets}
   Output Pose() {layerGroup=Item Override}
  ]}
 ItemFullBodyOverride {nodes [
   AnimGraphNode_StateMachine_0 = StateMachine() {is=title AnimGraphNode_StateMachine}
   n1 = Layered blend per bone(
    BasePose=@AnimGraphNode_StateMachine_0
    BlendPoses_0=Sequence Player 'DuckCoast_AngryFace'() {bLoopAnimation=false, sequence=/WildEstate_DuckCoast/Assets/Animation/Emotion/DuckCoast_AngryFace}
    BlendWeights_0=Get AngryFaceAlpha() -> (AngryFaceAlpha:float)
   ) {
    blendMode=BlendMask
    blendMasks [&Fortnite_M_Avg_Player_Skeleton:FacialCurvesNoWeights]
    layerSetup []
   }
   Output Pose(Result=@n1) {layerGroup=Item Override}
   Save cached pose 'FullBodyCache'(Pose=Input Pose() {name=InPoseFullBody}) {cacheName=FullBodyCache}
  ]}
 StateMachine {
  entry=UnawareLoop
  path=ItemFullBodyOverride.AnimGraphNode_StateMachine_0.StateMachine
  ids=id
  nodes [
   ToCombat() {is=title AnimStateAliasNode, also {aliases [ToAlert,ToUnaware,UnawareLoop]}}
   Alias() {is=title AnimStateAliasNode, also {aliases [Locomotion,UnawareLoop]}}
   AnimStateNode_0 = UnawareLoop() {is=title AnimStateNode}
   AnimStateNode_1 = AlertLoop() {is=title AnimStateNode}
   AnimStateNode_2 = ToAlert() {is=title AnimStateNode}
   AnimStateNode_3 = Combat() {is=title AnimStateNode}
   AnimStateNode_4 = ToUnaware() {is=title AnimStateNode}
   AnimStateNode_5 = Locomotion() {is=title AnimStateNode}
   AnimStateTransitionNode_0 = AnimStateTransitionNode() {bAllowInertializationForSelfTransitions=true}
   AnimStateTransitionNode_1 = AnimStateTransitionNode() {bAllowInertializationForSelfTransitions=true}
   AnimStateTransitionNode_12 = AnimStateTransitionNode() {bAllowInertializationForSelfTransitions=true}
   AnimStateTransitionNode_13 = AnimStateTransitionNode() {bAllowInertializationForSelfTransitions=true, bAutomaticRuleBasedOnSequencePlayerInState=true}
   AnimStateTransitionNode_14 = AnimStateTransitionNode() {bAllowInertializationForSelfTransitions=true}
   AnimStateTransitionNode_15 = AnimStateTransitionNode() {bAllowInertializationForSelfTransitions=true}
   AnimStateTransitionNode_17 = AnimStateTransitionNode() {bAllowInertializationForSelfTransitions=true}
   AnimStateTransitionNode_18 = AnimStateTransitionNode() {bAllowInertializationForSelfTransitions=true, bAutomaticRuleBasedOnSequencePlayerInState=true, crossfadeDuration=0.1}
   AnimStateTransitionNode_2 = AnimStateTransitionNode() {bAllowInertializationForSelfTransitions=true}
   AnimStateTransitionNode_23 = AnimStateTransitionNode() {bAllowInertializationForSelfTransitions=true, priorityOrder=2}
  ]
  links [
   {
    dst=Combat
    rule=ItemFullBodyOverride.AnimGraphNode_StateMachine_0.StateMachine.AnimStateTransitionNode_14.Transition
    src=AlertLoop
    via=AnimStateTransitionNode_14
   }
   {
    dst=ToUnaware
    rule=ItemFullBodyOverride.AnimGraphNode_StateMachine_0.StateMachine.AnimStateTransitionNode_17.Transition
    src=AlertLoop
    via=AnimStateTransitionNode_17
   }
   {
    dst=ToAlert
    rule=ItemFullBodyOverride.AnimGraphNode_StateMachine_0.StateMachine.AnimStateTransitionNode_2.Transition
    src=Alias
    via=AnimStateTransitionNode_2
   }
   {
    dst=AlertLoop
    rule=ItemFullBodyOverride.AnimGraphNode_StateMachine_0.StateMachine.AnimStateTransitionNode_15.Transition
    src=Combat
    via=AnimStateTransitionNode_15
   }
   {
    dst=UnawareLoop
    rule=ItemFullBodyOverride.AnimGraphNode_StateMachine_0.StateMachine.AnimStateTransitionNode_1.Transition
    src=Locomotion
    via=AnimStateTransitionNode_1
   }
   {
    dst=AlertLoop
    rule=ItemFullBodyOverride.AnimGraphNode_StateMachine_0.StateMachine.AnimStateTransitionNode_13.Transition
    src=ToAlert
    via=AnimStateTransitionNode_13
   }
   {
    dst=Combat
    rule=ItemFullBodyOverride.AnimGraphNode_StateMachine_0.StateMachine.AnimStateTransitionNode_23.Transition
    src=ToCombat
    via=AnimStateTransitionNode_23
   }
   {
    dst=UnawareLoop
    rule=ItemFullBodyOverride.AnimGraphNode_StateMachine_0.StateMachine.AnimStateTransitionNode_18.Transition
    src=ToUnaware
    via=AnimStateTransitionNode_18
   }
   {
    dst=Locomotion
    rule=ItemFullBodyOverride.AnimGraphNode_StateMachine_0.StateMachine.AnimStateTransitionNode_0.Transition
    src=UnawareLoop
    via=AnimStateTransitionNode_0
   }
   {
    dst=ToAlert
    rule=ItemFullBodyOverride.AnimGraphNode_StateMachine_0.StateMachine.AnimStateTransitionNode_12.Transition
    src=UnawareLoop
    via=AnimStateTransitionNode_12
   }
  ]
 }
 UnawareLoop {
  path=ItemFullBodyOverride.AnimGraphNode_StateMachine_0.StateMachine.AnimStateNode_0.UnawareLoop
  nodes [
   n1 = Sequence Player 'DuckCoast_Idle_CleaningLoop'() {groupName=Rag, groupRole=AlwaysLeader, method=SyncGroup, sequence=&/DuckCoast_Idle_CleaningLoop}
   Output Animation Pose(Result=Modify Curve(SourcePose=@n1, CurveValues_0=1, Alpha=1) {curveNames [UserCurve01], curveValues [0]})
  ]
 }
 AlertLoop {
  path=ItemFullBodyOverride.AnimGraphNode_StateMachine_0.StateMachine.AnimStateNode_1.AlertLoop
  nodes [
   n1 = Sequence Player 'DuckCoast_AlertedLoop'() {groupName=Loop, method=SyncGroup, sequence=&/DuckCoast_AlertedLoop}
   n2 = Use cached pose 'FullBodyCache'()
   Output Animation Pose(Result=Blend Poses by bool(
     BlendPose_0=Layered blend per bone(BasePose=@n2, BlendPoses_0=@n1) {layerSetup [{branchFilters [{boneName=Spine_01}]}]}
     BlendPose_1=Sequence Player 'DuckCoast_AlertedLoop'() {groupName=Loop, method=SyncGroup, sequence=&DuckCoast_AlertedLoop}
     BlendTime_0=0.2
     bActiveValue=Get bIsMoving2D()
    ))
  ]
 }
 ToAlert {
  path=ItemFullBodyOverride.AnimGraphNode_StateMachine_0.StateMachine.AnimStateNode_2.ToAlert
  nodes [
   n1 = Sequence Player 'DuckCoast_AlertedIntro'() {bLoopAnimation=false, groupName=AlertIntro, method=SyncGroup, sequence=&/DuckCoast_AlertedIntro}
   n2 = Use cached pose 'FullBodyCache'()
   n3 = Sequence Player 'DuckCoast_AlertedIntro'() {bLoopAnimation=false, groupName=AlertIntro, method=SyncGroup, sequence=&DuckCoast_AlertedIntro}
   Output Animation Pose(Result=Blend Poses by bool(
     BlendPose_0=Layered blend per bone(BasePose=@n2, BlendPoses_0=@n1) {layerSetup [{branchFilters [{boneName=Spine_01}]}]}
     BlendPose_1=@n3
     BlendTime_0=0.2
     bActiveValue=Get bIsMoving2D()
    ))
  ]
 }
 Combat {
  path=ItemFullBodyOverride.AnimGraphNode_StateMachine_0.StateMachine.AnimStateNode_3.Combat
  nodes [Output Animation Pose(Result=Use cached pose 'FullBodyCache'())]
 }
 ToUnaware {
  path=ItemFullBodyOverride.AnimGraphNode_StateMachine_0.StateMachine.AnimStateNode_4.ToUnaware
  nodes [
   n1 = Sequence Player 'DuckCoast_AlertedToIdle'() {bLoopAnimation=false, groupName=Alert, method=SyncGroup, sequence=&/DuckCoast_AlertedToIdle}
   n2 = Use cached pose 'FullBodyCache'()
   n3 = Sequence Player 'DuckCoast_AlertedToIdle'() {bLoopAnimation=false, groupName=Alert, method=SyncGroup, sequence=&DuckCoast_AlertedToIdle}
   Output Animation Pose(Result=Blend Poses by bool(
     BlendPose_0=Layered blend per bone(BasePose=@n2, BlendPoses_0=@n1) {layerSetup [{branchFilters [{boneName=Spine01}]}]}
     BlendPose_1=@n3
     BlendTime_0=0.2
     bActiveValue=Get bIsMoving2D()
    ))
  ]
 }
 Locomotion {
  path=ItemFullBodyOverride.AnimGraphNode_StateMachine_0.StateMachine.AnimStateNode_5.Locomotion
  nodes [Output Animation Pose(Result=Use cached pose 'FullBodyCache'())]
 }
 Transition {
  path=ItemFullBodyOverride.AnimGraphNode_StateMachine_0.StateMachine.AnimStateTransitionNode_0.Transition
  nodes [Result(bCanEnterTransition=Get bIsMoving2D())]
 }
 Transition {
  path=ItemFullBodyOverride.AnimGraphNode_StateMachine_0.StateMachine.AnimStateTransitionNode_1.Transition
  nodes [Result(bCanEnterTransition=NOT Boolean(A=Get bIsMoving2D()))]
 }
 Transition {
  path=ItemFullBodyOverride.AnimGraphNode_StateMachine_0.StateMachine.AnimStateTransitionNode_12.Transition
  nodes [Result(bCanEnterTransition=Get bAlerted() -> (bAlerted:bool))]
 }
 Transition {path=ItemFullBodyOverride.AnimGraphNode_StateMachine_0.StateMachine.AnimStateTransitionNode_13.Transition, nodes [Result()]}
 Transition {
  path=ItemFullBodyOverride.AnimGraphNode_StateMachine_0.StateMachine.AnimStateTransitionNode_14.Transition
  nodes [Result(bCanEnterTransition=Get bInCombat() -> (bInCombat:bool))]
 }
 Transition {
  path=ItemFullBodyOverride.AnimGraphNode_StateMachine_0.StateMachine.AnimStateTransitionNode_15.Transition
  nodes [Result(bCanEnterTransition=NOT Boolean(A=Get bInCombat()))]
 }
 Transition {
  path=ItemFullBodyOverride.AnimGraphNode_StateMachine_0.StateMachine.AnimStateTransitionNode_17.Transition
  nodes [Result(bCanEnterTransition=NOT Boolean(A=Get bAlerted()))]
 }
 Transition {path=ItemFullBodyOverride.AnimGraphNode_StateMachine_0.StateMachine.AnimStateTransitionNode_18.Transition, nodes [Result()]}
 Transition {
  path=ItemFullBodyOverride.AnimGraphNode_StateMachine_0.StateMachine.AnimStateTransitionNode_2.Transition
  nodes [Result(bCanEnterTransition=Get bAlerted())]
 }
 Transition {
  path=ItemFullBodyOverride.AnimGraphNode_StateMachine_0.StateMachine.AnimStateTransitionNode_23.Transition
  nodes [Result(bCanEnterTransition=Get bInCombat())]
 }
 ItemIdleOverride {nodes [
   Input Pose() {name=InPose_Idle}
   Output Pose() {layerGroup=Item Override}
  ]}
 ItemJetPackJumpAdditive {nodes [Output Pose() {layerGroup=Item Override}]}
 ItemJetPackStartAdditive {nodes [Output Pose() {layerGroup=Item Override}]}
 ItemJumpLoopAdditive {nodes [
   Input Pose() {name=InPoseItemJumpLoopAdditive}
   Output Pose() {layerGroup=Item Override}
  ]}
 ItemJumpOverride {nodes [
   Input Pose() {name=InPoseJump}
   Output Pose() {layerGroup=Item Override}
  ]}
 ItemJumpUpAdditive {nodes [
   Input Pose() {name=InPoseItemJumpUpAdditive}
   Output Pose() {layerGroup=Item Override}
  ]}
 ItemPreAimOffsets {nodes [
   Input Pose() {name=InPosePreAimOffsets}
   Output Pose() {layerGroup=Item Override}
  ]}
 ItemPreIK {nodes [
   Input Pose() {name=InPose_PreIK}
   Output Pose() {layerGroup=Item Override}
  ]}
 ItemRightHandIK {nodes [
   Input Pose() {name=InPosePreRightHandIK}
   Input Pose() {name=InPosePostRightHandIK}
   Output Pose() {layerGroup=Item Override}
  ]}
 ItemSkeletalControl {nodes [
   Input Pose()
   Output Pose() {layerGroup=Item Override}
  ]}
 ItemSwimJumpFallLoop {nodes [Output Pose() {layerGroup=Item Override}]}
 ItemSwimJumpSurfaceEnd {nodes [Output Pose() {layerGroup=Item Override}]}
 ItemSwimJumpSurfaceLoop {nodes [Output Pose() {layerGroup=Item Override}]}
 ItemUpperBody {nodes [
   Input Pose!() -> (Pose, InputParam:Fort Anim Input Adjusted Aim Structure) {name=InPoseUpperBody}
   Output Pose() {layerGroup=Item Override}
  ]}
 ItemZipLineStartAdditive {nodes [Output Pose() {layerGroup=Item Override}]}
]
variables []
